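# Translucent unlit material for the glowing strings effect of the GeneAglet bass cosmetic, flagged for Niagara mesh particles.
# SplineThicken sets the string width from a thickness parameter.
# A GeneratedBand forms the bright core, boosting the tint color before emissive adjustments.
# An opacity-style mask adds that band to a faint procedural halo, built from sine-shaped UV gradients that fade at the ends and edges.
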Material /SparksCosmetics/Instrument/Bass/GeneAglet/Effects/M_GeneAglet_Strings
graph {
 properties {
  bAutomaticallySetUsageInEditor=false
  bOutputTranslucentVelocity=true
  bUsedWithNiagaraMeshParticles=true
  blendMode=BLEND_Translucent
  shadingModel=MSM_Unlit
  usedShadingModels=Unlit
 }
 nodes [
  n1 = ScalarParameter() {defaultValue=3, parameterName=Spline WPO Thickness}
  MaterialFunctionCall(WidthBase=@n1, WidthTip=@n1, Expand U or V UV Channel=StaticBool()) {materialFunction=&SplineThicken}
  n2 = VectorParameter() {parameterName=Color, defaultValue=<rgba 0.1248 0.0144 0.5271 1>}
  n3 = Multiply(A=@n2.RGB, B=ScalarParameter() {defaultValue=3, parameterName=String Brightness Intensity}) {constB=5}
  n4 = MaterialFunctionCall!(
   Width=ScalarParameter() {defaultValue=0.04, parameterName=String Width}
   Sharpness=ScalarParameter() {defaultValue=0.75, parameterName=String Sharpness}
   Offset
   Direction Switch=StaticBool()
   Compare
   Input Coordinates
  ) {materialFunction=/Engine/Functions/Engine_MaterialFunctions03/Procedurals/GeneratedBand}
  MaterialFunctionCall(Emissive=Add(A=@n2.RGB, B=Multiply(A=@n3, B=@n4.Result))) {materialFunction=&MF_FX_EmissiveAdjustments}
  n5 = Sine(None=Multiply(A=MaterialFunctionCall() {materialFunction=&LinearGradient}.UGradient) {constB=0.5})
  n6 = SmoothStep(Max=ScalarParameter() {defaultValue=0.5, parameterName=Glow Ends Fade}, Value=Saturate(None=Power(Base=@n5) {constExponent=1})) {constMax=0.5}
  n7 = Sine(None=Multiply(A=MaterialFunctionCall() {materialFunction=&LinearGradient}.VGradient) {constB=0.5})
  n8 = Multiply(A=Saturate(None=Power(Base=@n7) {constExponent=6})) {constB=0.35}
  Saturate(None=Add(A=Multiply(A=@n6, B=@n8), B=@n4.Result))
 ]
}
