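# Niagara impact effect for the GeneAglet bass cosmetic, played as a short hit burst along a system direction vector.
# Burst emitters layer chroma meshes, flare sprites with a light, a HitColor-tinted star mesh, purple-pink-blue sparks and streaks, and surface-driven debris.
# FN_PhysicalSurface sets debris velocity, cone and gravity from User.PhysicalSurface.
# Front-end modifiers rescale parts for previews, while time-of-day inputs tone down the star and streaks at night.
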NiagaraSystem /SparksCosmetics/Instrument/Bass/GeneAglet/Effects/NS_Bass_GeneAglet_Impact
dependencies {
 usedDataInterfaces [
  &NiagaraDataInterfaceCurve
  &NiagaraDataInterfaceVector2DCurve
  &NiagaraDataInterfaceVectorCurve
  &NiagaraDataInterfaceColorCurve
  &NiagaraDataInterfaceArrayFloat
  &NiagaraDataInterfaceArrayFloat3
 ]
 usedDynamicInputs [
  &Vector2DFromFloat
  /Niagara/DynamicInputs/Math/Clamp_Vector
  &SelectVectorFromArray
 ]
 usedModules [
  &SystemState
  &self:SystemSpawnScript.SystemScriptSource.SystemScriptGraph.NiagaraNodeAssignment_0.SetVariables_#1
 ]
}
emitters {
 AudioChroma {
  emitterUpdateScript [
   &EmitterState()
   &SpawnBurst_Instantaneous(Spawn Count=2, Use Spawn Probability=false, UseLoopCountLimit=false)
  ]
  particleSpawnScript [
   &InitializeParticle(
    Lifetime=0.175
    Mesh Scale Mode=::NewEnumerator3
    Mesh Uniform Scale=&SelectFloatFromArray(
     Direct Array Index=&ReturnExecIndex()
     Float Selection Array=DataInterface() {propertyValues {FloatData [0.75,0.5]}}
    )
    UsePositionOffset=false
    Write Material Random=true
   )
   &AddVelocity(
    Velocity=&Add_Vector(A=$System.DirectionVector, B=&RandomVector(Vector Scale=0))
    Velocity Speed Scale=10
   )
   &InitialMeshOrientation(Mesh Orientation Mode=::NewEnumerator2, Rotation=<0 -0.25 0>, Use Rotation Vector=true) {enabled=false}
   &CameraOffset(Camera Offset Amount=30)
  ]
  particleUpdateScript [
   &ParticleState()
   &ScaleMeshSize(Scale Factor=&Multiply_VectorByFloat(
     Vector=&MakeVector(
      X=&FloatFromCurve(
       FloatCurve=DataInterface() {propertyValues {Curve {keys [{arriveTangent=2.9252, interpMode=RCIM_Cubic, leaveTangent=2.9252, tangentMode=RCTM_User, time=0, value=0}, <at 1 1>]}}}
       Scale Curve=1.5
      )
      Y=1
      Z=1
     )
     Float=&FN_FrontEndFloatModifier(Gameplay=1, Front End=0.6)
    ))
   &SolveForcesAndVelocity(Clamp Velocity=false, Limit Acceleration=false)
   &DynamicMaterialParameters(
    Index 0 Param 3=&FloatFromCurve(FloatCurve=DataInterface() {propertyValues {Curve {keys [<at 0 0>,<at 1 10>]}}})
    Index 0 Param 4=&FloatFromCurve(
     FloatCurve=DataInterface() {propertyValues {Curve {keys [<at 0 0.25>,<at 1 1.5>]}}}
     Scale Curve=&FN_FrontEndFloatModifier(Gameplay=1, Front End=0.6)
    )
    Module.Param0WriteEnabled=false
    Module.Param1WriteEnabled=false
    Module.Param2WriteEnabled=true
    Module.Param3WriteEnabled=true
   )
   &ScaleColor(
    Scale Alpha=&FloatFromCurve(FloatCurve=DataInterface() {propertyValues {Curve {keys [
         {interpMode=RCIM_Cubic, leaveTangentWeight=0.0667, time=0, value=1}
         {
          arriveTangent=-1.9874
          arriveTangentWeight=2.0817
          interpMode=RCIM_Cubic
          leaveTangent=-1.9874
          leaveTangentWeight=1.249
          tangentMode=RCTM_User
          time=0.2
          value=0.5
         }
         {arriveTangent=-0.625, interpMode=RCIM_Cubic, leaveTangent=-0.625, time=0.5, value=0.1}
         {arriveTangentWeight=0.2333, interpMode=RCIM_Cubic, time=1, value=0}
        ]}}})
    ScaleA=true
    ScaleRGB=false
   )
   &ScaleColor(
    Scale Alpha=&FloatFromCurve(FloatCurve=DataInterface() {propertyValues {Curve {keys [
         {
          arriveTangent=-2.6331
          arriveTangentWeight=2.0817
          interpMode=RCIM_Cubic
          leaveTangent=-2.6331
          leaveTangentWeight=1.249
          tangentMode=RCTM_User
          time=0
          value=1
         }
         {arriveTangent=-0.3634, interpMode=RCIM_Cubic, leaveTangent=-0.3634, tangentMode=RCTM_User, time=0.2, value=0.1}
         {arriveTangent=-0.0099, tangentMode=RCTM_User, time=1, value=0}
        ]}}})
    ScaleA=true
    ScaleRGB=false
   ) {enabled=false, module=ScaleColor001}
  ]
  rendererClasses [&NiagaraMeshRendererProperties]
  renderers [{rendererClass=&NiagaraMeshRendererProperties, rendererIndex=0}]
 }
 Flash_Flare001 {
  emitterUpdateScript [
   &EmitterState()
   &SpawnBurst_Instantaneous(Use Spawn Probability=false, UseLoopCountLimit=false)
  ]
  particleSpawnScript [
   &InitializeParticle(
    Color=<rgba 0.5327 0 2 1>
    Lifetime=0.18
    Sprite Size=<80 240>
    Sprite Size Mode=::NewEnumerator4
    UsePositionOffset=false
    Write Material Random=true
   )
   &CameraOffset(Camera Offset Amount=20)
   &DynamicMaterialParameters(
    Index 3 Param 1=1.5
    Index 3 Param 2=1
    Module.Param0WriteEnabled=true
    Module.Param0WriteEnabled3=true
    Module.Param1WriteEnabled=true
    Module.Param1WriteEnabled3=true
    Module.Param2WriteEnabled=true
    Module.Param2WriteEnabled3=true
    Module.Param3WriteEnabled=true
    Module.Param3WriteEnabled3=true
    Write Parameter Index 3=true
   )
   &AddVelocity(Velocity=$System.DirectionVector, Velocity Speed Scale=&RandomRangeFloat(Minimum=200, Maximum=300))
  ]
  particleUpdateScript [
   &ParticleState()
   &ScaleSpriteSize(
    Non-Uniform Curve Sprite Scale=DataInterface() {propertyValues {
      XCurve {keys [
        {arriveTangent=7.4419, interpMode=RCIM_Cubic, leaveTangent=7.4419, tangentMode=RCTM_User, time=0, value=0}
        {interpMode=RCIM_Cubic, tangentMode=RCTM_User, time=0.2, value=1.075}
        {interpMode=RCIM_Cubic, tangentMode=RCTM_User, time=1, value=-0.1}
       ]}
      YCurve {keys [
        {arriveTangent=19.7829, interpMode=RCIM_Cubic, leaveTangent=19.7829, tangentMode=RCTM_User, time=0, value=0}
        {interpMode=RCIM_Cubic, tangentMode=RCTM_User, time=0.1, value=1.025}
        {interpMode=RCIM_Cubic, tangentMode=RCTM_User, time=1, value=0.8}
       ]}
     }}
    Scale Sprite Size Mode=::NewEnumerator3
   )
   &SolveForcesAndVelocity(Clamp Velocity=false, Limit Acceleration=false)
  ]
  rendererClasses [&NiagaraSpriteRendererProperties]
  renderers [{rendererClass=&NiagaraSpriteRendererProperties}]
 }
 Flash_Flare_Light {
  emitterUpdateScript [
   &EmitterState()
   &SpawnBurst_Instantaneous(Use Spawn Probability=false, UseLoopCountLimit=false)
  ]
  particleSpawnScript [
   &InitializeParticle(
    Color=<rgba 0.883 0.0125 1.25 1>
    Lifetime=0.25
    Sprite Size Mode=::NewEnumerator3
    Uniform Sprite Size=120
    UsePositionOffset=false
    Write Material Random=true
   )
   &DynamicMaterialParameters(
    Index 0 Param 1=0.2
    Index 0 Param 2=1
    Module.Param0WriteEnabled=true
    Module.Param1WriteEnabled=true
    Module.Param2WriteEnabled=true
    Module.Param3WriteEnabled=true
   ) {enabled=false}
   &SystemLocation(MaskSpawnGroup=false, Offset=$System.HitOffset)
   &AddVelocity(Velocity=$System.DirectionVector)
  ]
  particleUpdateScript [
   &ParticleState()
   &ScaleSpriteSize(
    Uniform Curve Scale=&FN_FrontEndFloatModifier(Gameplay=1, Front End=$User.FrontEndScale)
    Uniform Curve Sprite Scale=DataInterface() {propertyValues {Curve {keys [
        {arriveTangent=3.5908, interpMode=RCIM_Cubic, leaveTangent=3.5908, tangentMode=RCTM_User, time=0, value=0}
        {tangentMode=RCTM_User, time=0.8, value=1}
       ]}}}
   )
   &ScaleColor(
    Scale Alpha=&FloatFromCurve(FloatCurve=DataInterface() {propertyValues {Curve {keys [
         {arriveTangent=-5.6669, interpMode=RCIM_Cubic, leaveTangent=-5.6669, tangentMode=RCTM_User, time=0, value=2}
         {arriveTangent=-1.179, interpMode=RCIM_Cubic, leaveTangent=-1.179, tangentMode=RCTM_User, time=0.5, value=0.5}
         {arriveTangent=-0.8686, interpMode=RCIM_Cubic, leaveTangent=-0.8686, tangentMode=RCTM_User, time=1, value=0}
        ]}}})
    Scale RGB=&VectorFromFloat(Value=&FN_FrontEndFloatModifier(Gameplay=1.5, Front End=1))
    ScaleA=true
    ScaleRGB=true
   )
   &FN_Light_Attributes(
    Color=$Particles.Color
    Create Color Attribute=true
    Create Exponent Attribute=true
    Create Position Attribute=true
    Create Radius Attribute=true
    Exponent=5
    Position=<15 0 0>
    Radius=250
    Scale Color=&FloatFromCurve(
     FloatCurve=DataInterface() {propertyValues {Curve {keys [
         {interpMode=RCIM_Cubic, tangentMode=RCTM_User, time=0, value=1}
         {arriveTangent=-0.5426, interpMode=RCIM_Cubic, leaveTangent=-0.5426, tangentMode=RCTM_User, time=1, value=0}
        ]}}}
     Scale Curve=5
    )
   )
   &CameraOffset(Camera Offset Amount=&Lerp_Float(A=15, B=4, Alpha=&MakeCustomFloatFromBool(Bool=$User.bIsFrontEndPreview, True Float=1, False Float=0)))
   /SetVariables(
    Particles.DarkBackerColor:LinearColor=&MultiplyLinearColor(
     Linear Color=<rgba 0.0087 0 0.3115 1>
     Scale Factor=&MakeVector4FromVectorAndFloat(
      "Vector (RGB)"=&VectorFromFloat(Value=&Lerp_Float(A=1, B=3, Alpha=$Particles.NormalizedAge))
      "Float (Alpha)"=&FloatFromCurve(FloatCurve=DataInterface() {propertyValues {Curve {keys [
           {arriveTangent=-17.4607, interpMode=RCIM_Cubic, leaveTangent=-17.4607, tangentMode=RCTM_User, time=0, value=5}
           {arriveTangent=-2.7181, tangentMode=RCTM_User, time=0.8, value=0}
          ]}}})
     )
    )
    Particles.DarkBackerOffset:NiagaraFloat=&Add_Float(A=$Particles.CameraOffset, B=-2)
    Particles.DarkBackerScale:Vector2f=&Multiply_Vector2D(A=$Particles.SpriteSize, B=<3.2 1.6>)
    Particles.DarkBackerVelocity:Vector3f=$System.DirectionVector
    Particles.DarkBackerVisibility:NiagaraInt32=&SpawnGroup()
   ) {isSetParameters=true}
   &SolveForcesAndVelocity(Clamp Velocity=false, Limit Acceleration=false)
  ]
  rendererClasses [&NiagaraSpriteRendererProperties, &NiagaraLightRendererProperties]
  renderers [
   {rendererClass=&NiagaraSpriteRendererProperties}
   {rendererClass=&NiagaraLightRendererProperties, rendererIndex=1}
   {rendererClass=&NiagaraSpriteRendererProperties, rendererIndex=2}
  ]
 }
 Flash_Star {
  emitterUpdateScript [
   &EmitterState(
    Enable Distance Culling=true
    Max Distance Response=::KillImmediately
    Module.bUseMaxDistance=true
    Module.bUseMinDistance=false
    Scalability Mode=::NewEnumerator1
   )
   &SpawnBurst_Instantaneous(Use Spawn Probability=false, UseLoopCountLimit=false)
  ]
  particleSpawnScript [
   &InitializeParticle(
    Color=&MultiplyLinearColor(Linear Color=<rgba 1 1 1 1>, Scale Factor=&MakeVector4_FromLinearColor(LinearColor=$User.HitColor))
    Lifetime=0.15
    Mesh Scale=&Multiply_VectorByFloat(
     Vector=<0.001 0.35 0.35>
     Float=&MakeCustomFloatFromBool(Bool=$User.bIsFrontEndPreview, True Float=$User.FrontEndScale, False Float=1)
    )
    Mesh Scale Mode=::NewEnumerator4
    UsePositionOffset=false
    Write Material Random=true
   )
   &AddVelocity(Velocity=$System.DirectionVector, Velocity Speed Scale=75)
  ]
  particleUpdateScript [
   &ParticleState()
   &ScaleMeshSize()
   &ScaleColor(
    Scale Alpha=&FloatFromCurve(
     FloatCurve=DataInterface() {propertyValues {Curve {keys [
         {arriveTangent=-23.7811, interpMode=RCIM_Cubic, leaveTangent=-23.7811, tangentMode=RCTM_User, time=0, value=5}
         {arriveTangent=-0.7993, interpMode=RCIM_Cubic, leaveTangent=-0.7993, tangentMode=RCTM_User, time=0.25, value=1}
         {arriveTangentWeight=0.4344, interpMode=RCIM_Cubic, tangentMode=RCTM_User, time=1, value=0}
        ]}}}
     Scale Curve=&FN_Select_Float_from_Time_Of_Day(Morning=1, Day=1, Evening=1, Night=0.25)
    )
    Scale RGB=&Multiply_Vector(
     A=&Multiply_VectorByFloat(
      Vector=&MakeVectorFromLinearColorRGB(Color=&ColorFromCurve(ColorCurve=DataInterface() {propertyValues {
          AlphaCurve {keys [<at 0 1 RCIM_Cubic>,<at 0.9977 1 RCIM_Cubic>]}
          BlueCurve {keys [<at 0.0057 0.9583>,<at 0.2851 1.0036>,<at 0.4798 1>,<at 0.6167 1>,<at 0.7729 0.9925>,<at 0.8885 0.8013>,<at 0.9448 0.5573>]}
          GreenCurve {keys [<at 0.0057 0.9583>,<at 0.2851 0.955>,<at 0.4798 0.5751>,<at 0.6167 0.1713>,<at 0.7729 0.2424>,<at 0.8885 0.6295>,<at 0.9448 1>]}
          RedCurve {keys [<at 0.0057 1>,<at 0.2851 0.2661>,<at 0.4798 0.2342>,<at 0.6167 0.1562>,<at 0.7729 0.5041>,<at 0.8885 0.8565>,<at 0.9448 0.5695>]}
         }}))
      Float=&FloatFromCurve(FloatCurve=DataInterface() {propertyValues {Curve {keys [{arriveTangent=-11.4904, interpMode=RCIM_Cubic, leaveTangent=-11.4904, tangentMode=RCTM_User, time=0, value=5}, <at 1 1>]}}})
     )
     B=&FN_Select_Vector_from_Time_Of_Day(Morning=<1 1 1>, Day=<1 1 1>, Evening=<1 1 1>, Night=<0.8 0.8 1.3>)
    )
    ScaleA=true
    ScaleRGB=true
   )
   &DynamicMaterialParameters(
    Index 0 Param 2=-1
    Index 0 Param 3=5
    Index 0 Param 4=1
    Module.Param0WriteEnabled=true
    Module.Param1WriteEnabled=true
    Module.Param2WriteEnabled=true
    Module.Param3WriteEnabled=true
   )
   &CameraOffset(Camera Offset Amount=&Lerp_Float(A=100, B=1, Alpha=&MakeCustomFloatFromBool(Bool=$User.bIsFrontEndPreview, True Float=1, False Float=0)))
   &SolveForcesAndVelocity(Clamp Velocity=false, Limit Acceleration=false)
  ]
  rendererClasses [&NiagaraMeshRendererProperties]
  renderers [{rendererClass=&NiagaraMeshRendererProperties, rendererIndex=0}]
 }
 NE_FN_PhysMatDebris {
  emitterUpdateScript [
   &EmitterState()
   &SpawnBurst_Instantaneous(
    Spawn Count=&RandomRangeInt(Minimum=5, Maximum=10)
    Use Spawn Probability=false
    UseLoopCountLimit=false
   )
  ]
  particleSpawnScript [
   /SetVariables(Particles.BiasedRandom=&FloatFromCurve(
     FloatCurve=DataInterface() {propertyValues {Curve {keys [
         <at 0.2 1 RCIM_Cubic>
         {arriveTangent=-3.3394, interpMode=RCIM_Cubic, leaveTangent=-3.3394, tangentMode=RCTM_User, time=0.5, value=0.5}
         <at 0.8 0 RCIM_Cubic>
        ]}}}
     CurveIndex=&RandomRangeFloat(Minimum=0, Maximum=1)
    )) {isSetParameters=true}
   &InitializeParticle(
    Color Mode=::NewEnumerator0
    Lifetime=&RandomRangeFloat(Minimum=0.5, Maximum=1.25)
    Mass=&Lerp_Float(A=0.6, B=1.2, Alpha=$Particles.BiasedRandom)
    Mass Mode=::NewEnumerator1
    Mesh Scale=&Multiply_VectorByFloat(
     Vector=&RandomRangeVector(Minimum=<0.9 0.4 0.5>, Maximum=<1 1 1>)
     Float=&Lerp_Float(A=0.7, B=1, Alpha=$Particles.BiasedRandom)
    )
    Mesh Scale Mode=::NewEnumerator4
    UsePositionOffset=false
    Write Material Random=true
   )
   &FN_PhysicalSurface(
    Dirt Mesh Count=2
    Glass Mesh Count=3
    Ice Mesh Count=2
    Metal Gravity=&RandomRangeFloat(Minimum=-350, Maximum=-1800)
    Metal Mesh Count=2
    Metal Velocity=&RandomRangeFloat(Minimum=80, Maximum=375)
    Metal Velocity Cone Angle=130
    Physical Surface=$User.PhysicalSurface
    Stone Gravity=&RandomRangeFloat(Minimum=-700, Maximum=-1250)
    Stone Mesh Count=2
    Stone Velocity=&RandomRangeFloat(Minimum=80, Maximum=550)
    Stone Velocity Cone Angle=120
    Wood Gravity=&RandomRangeFloat(Minimum=-450, Maximum=-900)
    Wood Mesh Count=2
    Wood Scale=1.5
    Wood Velocity=&RandomRangeFloat(Minimum=80, Maximum=600)
    Wood Velocity Cone Angle=130
   )
   &AddVelocityInCone(
    Cone Angle=$Output.FN_PhysicalSurface.SurfaceResponseConeAngle
    Module.Use Velocity Falloff On Cone Axis=false
    Velocity Strength=$Output.FN_PhysicalSurface.SurfaceResponseVelocity
   )
   &InitialMeshOrientation(Use Rotation Vector=false)
   &AddRotationalVelocity(Rotation Rate=&RandomVector(Vector Scale=&RandomRangeFloat(Minimum=1, Maximum=2)))
   &ApplyInitialForces()
   /SetVariables(Particles.GravityOverrideValue=$Output.FN_PhysicalSurface.GravityOverride) {isSetParameters=true}
  ]
  particleUpdateScript [
   &ParticleState()
   &GravityForce(Gravity=$Particles.GravityOverrideValue)
   &Collision(
    Advanced Aging Rate=1
    Age Colliding Particles=false
    CPU Collision Trace Channel=ECC_WorldStatic
    Collision Enabled=&Boolean_AndOperation(
     Bool1=&SetBoolByFloatComparison(A=$Particles.Initial.Lifetime, B=1)
     Bool2=&SetBoolByFloatComparison(A=$Particles.Age, B=0.12)
    )
    Control Roll On Collision=true
    Friction=1
    Friction During a Bounce=0.25
    Max CPU Trace Length=100
    Module.Control Rest State Via Surface Normal=false
    Module.EnableMaxCollisionCount=false
    Module.Randomize Collision Normal Vector=true
    Randomize Collision Normal=0.25
    ScaleRateOfRotation=false
   ) {enabled=false}
   &SolveForcesAndVelocity(Clamp Velocity=false, Limit Acceleration=false)
   &ScaleMeshSize(Scale Factor=&VectorFromFloat(Value=&FloatFromCurve(FloatCurve=DataInterface() {propertyValues {Curve {keys [<at 0.6 1 RCIM_Cubic>, {arriveTangent=-6.308, interpMode=RCIM_Cubic, leaveTangent=-6.308, tangentMode=RCTM_User, time=1, value=0}]}}})))
   &DynamicMaterialParameters(Module.Param0WriteEnabled=true, Module.Param1WriteEnabled=true, Module.Param2WriteEnabled=false, Module.Param3WriteEnabled=false)
  ]
  rendererClasses [&NiagaraSpriteRendererProperties, &NiagaraMeshRendererProperties]
  renderers [
   {rendererClass=&NiagaraSpriteRendererProperties}
   {rendererClass=&NiagaraMeshRendererProperties, rendererIndex=1}
   {rendererClass=&NiagaraMeshRendererProperties, rendererIndex=2}
   {rendererClass=&NiagaraMeshRendererProperties, rendererIndex=3}
   {rendererClass=&NiagaraMeshRendererProperties, rendererIndex=4}
   {rendererClass=&NiagaraMeshRendererProperties, rendererIndex=5}
   {rendererClass=&NiagaraMeshRendererProperties, rendererIndex=6}
   {rendererClass=&NiagaraMeshRendererProperties, rendererIndex=7}
   {rendererClass=&NiagaraMeshRendererProperties, rendererIndex=8}
  ]
 }
 Sparks {
  emitterUpdateScript [
   &EmitterState()
   &SpawnBurst_Instantaneous(Spawn Count=35, Use Spawn Probability=false, UseLoopCountLimit=false)
   &SpawnBurst_Instantaneous(Spawn Count=5, Spawn Group=1, Use Spawn Probability=false, UseLoopCountLimit=false) {module=SpawnBurst_Instantaneous001}
  ]
  particleSpawnScript [
   &InitializeParticle(
    Color Maximum=<rgba 2 0.1 1.1508 1>
    Color Minimum=<rgba 0.3929 0.0625 1.25 1>
    Color Mode=::NewEnumerator2
    Lifetime=&FloatFromCurve(
     FloatCurve=DataInterface() {propertyValues {Curve {keys [
         {arriveTangent=-5.4544, interpMode=RCIM_Cubic, leaveTangent=-5.4544, tangentMode=RCTM_User, time=0, value=0.75}
         {arriveTangent=-0.65, interpMode=RCIM_Cubic, leaveTangent=-0.65, time=0.1, value=0.4}
         <at 1 0.1>
        ]}}}
     CurveIndex=&RandomRangeFloat(Minimum=0, Maximum=1)
    )
    Sprite Size Mode=::NewEnumerator3
    Uniform Sprite Size=&Multiply_Float(
     A=&FN_FrontEndFloatModifier(Gameplay=1, Front End=0.6)
     B=&RandomRangeFloat(Minimum=8, Maximum=14)
    )
    UsePositionOffset=false
    Write Material Random=true
   )
   &ShapeLocation(
    MaskSpawnGroup=false
    Override Seed=false
    Rotation Axis=$System.DirectionVector
    Rotation Mode=::NewEnumerator1
    Sphere Radius=5
    UseOwnerScale=true
   )
   &CameraOffset(Camera Offset Amount=10)
   &AddVelocityFromPoint(
    Module.Use Falloff Distance=false
    Velocity Strength=&Multiply_Float(
     A=&RandomRangeFloat(Minimum=150, Maximum=200)
     B=&FN_FrontEndFloatModifier(Gameplay=&MaskFloatBySpawnGroup(Masked Float=1, Passthrough Float=3), Front End=0.5)
    )
   )
   &AddVelocity(
    Velocity=$System.DirectionVector
    Velocity Speed Scale=&Multiply_Float(
     A=&RandomRangeFloat(Minimum=300, Maximum=700)
     B=&FN_FrontEndFloatModifier(Gameplay=&MaskFloatBySpawnGroup(Masked Float=1, Passthrough Float=0), Front End=$User.FrontEndScale)
    )
   )
  ]
  particleUpdateScript [
   &ParticleState()
   &AccelerationForce(Acceleration=&RandomRangeVector(Minimum=<0 0 -2000>, Maximum=<0 0 -1000>))
   &Drag(Drag=&RandomRangeFloat(Minimum=1.5, Maximum=3), Use Linear Drag=true, Use Rotational Drag=true)
   &SolveForcesAndVelocity(Clamp Velocity=false, Limit Acceleration=false)
   &ScaleSpriteSizeBySpeed(Max Scale Factor=<0.1 2>, Min Scale Factor=<0.25 0.2>)
   &ScaleSpriteSize(
    Uniform Curve Scale=&FN_FrontEndFloatModifier(Gameplay=1, Front End=$User.FrontEndScale)
    Uniform Curve Sprite Scale=DataInterface() {propertyValues {Curve {keys [{arriveTangent=-6.725, interpMode=RCIM_Cubic, leaveTangent=-6.725, tangentMode=RCTM_User, time=0.65, value=1}, <at 1 0 RCIM_Cubic>]}}}
   )
   &ScaleColor(
    Scale Alpha=&FloatFromCurve(
     FloatCurve=DataInterface() {propertyValues {Curve {keys [
         {interpMode=RCIM_Cubic, tangentMode=RCTM_User, time=0, value=1}
         {arriveTangent=-0.3845, interpMode=RCIM_Cubic, leaveTangent=-0.3845, tangentMode=RCTM_User, time=0.5, value=0.9}
         {arriveTangent=-0.2828, interpMode=RCIM_Cubic, leaveTangent=-0.2828, tangentMode=RCTM_User, time=1, value=0}
        ]}}}
     Scale Curve=&RandomRangeFloat(Minimum=0.25, Maximum=5)
    )
    Scale RGB=&Multiply_VectorByFloat(
     Vector=&MakeVectorFromLinearColorRGB(Color=&ColorFromCurve(ColorCurve=DataInterface() {propertyValues {
         AlphaCurve {keys [<at 0 1 RCIM_Cubic>]}
         BlueCurve {keys [<at 0.1873 2 RCIM_Cubic>,<at 0.5469 1 RCIM_Cubic>]}
         GreenCurve {keys [<at 0.1873 2 RCIM_Cubic>,<at 0.5469 0.1339 RCIM_Cubic>]}
         RedCurve {keys [<at 0.1873 2 RCIM_Cubic>,<at 0.5469 0.1457 RCIM_Cubic>]}
        }}))
     Float=20
    )
    ScaleA=true
    ScaleRGB=true
   )
  ]
  rendererClasses [&NiagaraSpriteRendererProperties]
  renderers [{rendererClass=&NiagaraSpriteRendererProperties}]
 }
 Streaks {
  emitterUpdateScript [
   &EmitterState()
   &SpawnBurst_Instantaneous(Spawn Count=3, Use Spawn Probability=false, UseLoopCountLimit=false)
   &SpawnBurst_Instantaneous(Spawn Count=3, Spawn Group=1, Use Spawn Probability=true, UseLoopCountLimit=false) {module=SpawnBurst_Instantaneous001}
  ]
  particleSpawnScript [
   &InitializeParticle(
    Color Maximum=&MaskLinearColorBySpawnGroup(Masked Linear Color=<rgba 1 0.05 0.5754 1>, Passthrough Linear Color=<rgba 0.1 0.2057 2 1>)
    Color Minimum=<rgba 0.2242 0.05 1 1>
    Color Mode=::NewEnumerator2
    Lifetime Max=0.15
    Lifetime Min=0.1
    Lifetime Mode=::NewEnumerator1
    Sprite Size=&Multiply_Vector2D_ByFloat(
     Float=&FN_FrontEndFloatModifier(Gameplay=1, Front End=0.5)
     Vector2D=&UniformRangedVector2D(Minimum=<20 40>, Maximum=<25 70>)
    )
    Sprite Size Mode=::NewEnumerator4
    Sprite UV Mode=::NewEnumerator1
    UsePositionOffset=false
    Write Material Random=true
   ) {module=InitializeParticle001}
   &AddVelocity(
    Module.Use Falloff Distance=false
    Override Seed=false
    Velocity Mode=::NewEnumerator1
    Velocity Speed=&Multiply_Float(A=25, B=&FN_FrontEndFloatModifier(Gameplay=1, Front End=$User.FrontEndScale))
   ) {module=AddVelocity001}
   &SubUVAnimation(End Frame=3, SubUV Animation Mode:ENiagara_SubUVLookupMode=::NewEnumerator2)
   &AddVelocity(
    Velocity=$System.DirectionVector
    Velocity Speed Scale=&MaskFloatBySpawnGroup(Masked Float=100, Passthrough Float=50)
   )
   &CameraOffset(Camera Offset Amount=16)
  ]
  particleUpdateScript [
   &ParticleState()
   &SolveForcesAndVelocity(Clamp Velocity=false, Limit Acceleration=false)
   &ScaleSpriteSize(Scale Factor=&Vector2DFromCurve(
     Vector2Curve=DataInterface() {propertyValues {
       XCurve {keys [
         <at 0 1 RCIM_Cubic>
         {arriveTangent=-1, interpMode=RCIM_Cubic, leaveTangent=-1, time=0.5, value=0.5}
         {arriveTangent=-1.9017, interpMode=RCIM_Cubic, leaveTangent=-1.9017, tangentMode=RCTM_User, time=1, value=0}
        ]}
       YCurve {keys [{arriveTangent=3.9291, interpMode=RCIM_Cubic, leaveTangent=3.9291, tangentMode=RCTM_User, time=0, value=0}, <at 1 2 RCIM_Cubic>]}
      }}
     ScaleCurve=&MaskVector2dBySpawnGroup(Masked Vector2D=<1.5 1>, Passthrough Vector2D=<1.5 0.8>, Spawn Group Mask=0)
    ))
   &ScaleColor(
    Scale Alpha=&FloatFromCurve(
     FloatCurve=DataInterface() {propertyValues {Curve {keys [<at 0 1 RCIM_Cubic>, {arriveTangent=-3.2219, tangentMode=RCTM_User, time=1, value=0}]}}}
     Scale Curve=&FNTimeOfDayScalar(Night Scale=0.5, Morning Scale=1, Evening Scale=0.5)
    )
    Scale RGB=&Multiply_VectorByFloat(
     Vector=&VectorFromCurve(VectorCurve=DataInterface() {propertyValues {XCurve {keys [<at 0 1>,<at 1 0.75>]}, YCurve {keys [<at 0 1>,<at 1 0.5>]}, ZCurve {keys [<at 0 1>,<at 1 1>]}}})
     Float=&FloatFromCurve(FloatCurve=DataInterface() {propertyValues {Curve {keys [
          {arriveTangent=-32.1379, interpMode=RCIM_Cubic, leaveTangent=-32.1379, tangentMode=RCTM_User, time=0, value=10}
          {arriveTangent=-9, interpMode=RCIM_Cubic, leaveTangent=-9, time=0.2879, value=4.4753}
          {arriveTangent=-7.996, interpMode=RCIM_Cubic, leaveTangent=-7.996, tangentMode=RCTM_User, time=1, value=1}
         ]}}})
    )
    ScaleA=true
    ScaleRGB=true
   )
  ]
  rendererClasses [&NiagaraSpriteRendererProperties]
  renderers [{rendererClass=&NiagaraSpriteRendererProperties}]
 }
}
summary {
 systemName=NS_Bass_GeneAglet_Impact
 emitters [
  {
   emitterName=NE_FN_PhysMatDebris
   rendererClasses [&NiagaraSpriteRendererProperties, &NiagaraMeshRendererProperties]
  }
  {
   emitterName=Flash_Flare_Light
   rendererClasses [&NiagaraSpriteRendererProperties, &NiagaraLightRendererProperties]
  }
  {emitterName=Streaks, rendererClasses [&NiagaraSpriteRendererProperties]}
  {emitterName=Sparks, rendererClasses [&NiagaraSpriteRendererProperties]}
  {emitterName=Flash_Flare001, rendererClasses [&NiagaraSpriteRendererProperties]}
  {emitterName=AudioChroma, rendererClasses [&NiagaraMeshRendererProperties]}
  {emitterName=Flash_Star, rendererClasses [&NiagaraMeshRendererProperties]}
 ]
 userVariables [
  User.bIsFrontEnd {type=NiagaraBool}
  User.bIsFrontEndPreview {type=NiagaraBool}
  User.FrontEndScale {type=NiagaraFloat, defaultValue=0.8, description=""}
  User.HitColor {type=LinearColor, description="", defaultValue {struct=LinearColor, value=<rgba 0.2634 0.9002 1 1>}}
  User.PhysicalSurface {type=NiagaraInt32}
  User.Skeletal Mesh {
   type=NiagaraDataInterfaceSkeletalMesh
   defaultValue {
    struct=NiagaraExt_VariableValue_DataInterface
    value {dataInterface=&self:NiagaraDataInterfaceSkeletalMesh_0}
   }
  }
  User.SwingSectionIndex {type=NiagaraInt32}
 ]
}
