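# Niagara effect for the strings of the GeneAglet bass instrument cosmetic, adding a purple glow.
# The Light emitter burst-spawns at a small position offset, with light and sprite renderers and a shared LightRadius attribute driving sprite size.
# Particles loop their lifetime instead of expiring and can die when the emitter deactivates.
# A second LightweightStrings emitter is listed with no scripts or renderers shown.
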NiagaraSystem /SparksCosmetics/Instrument/Bass/GeneAglet/Effects/NS_Bass_GeneAglet_Strings
dependencies {usedDataInterfaces [], usedDynamicInputs [], usedModules [&SystemState]}
emitters {
 Light {
  emitterUpdateScript [
   &EmitterState()
   &SpawnBurst_Instantaneous(Use Spawn Probability=false, UseLoopCountLimit=true)
  ]
  particleSpawnScript [
   /SetVariables(Particles.LightRadius=20) {isSetParameters=true}
   &InitializeParticle(
    Color=<rgba 0.4734 0.0548 2 1>
    Position Offset=<0 10 5>
    Sprite Size Mode=::NewEnumerator3
    Uniform Sprite Size=$Particles.LightRadius
    UsePositionOffset=true
    Write Material Random=true
   )
   /SetVariables(Particles.SpriteSize=&Vector2DFromFloat(Value=$Particles.LightRadius)) {isSetParameters=true}
  ]
  particleUpdateScript [&ParticleState(
    Kill Particles When Lifetime Has Elapsed=false
    Let Infinitely Lived Particles Die When Emitter Deactivates=true
    Loop Particles Lifetime=true
   )]
  rendererClasses [&NiagaraLightRendererProperties, &NiagaraSpriteRendererProperties]
  renderers [
   {rendererClass=&NiagaraLightRendererProperties, rendererIndex=0}
   {rendererClass=&NiagaraSpriteRendererProperties, rendererIndex=1}
  ]
 }
 LightweightStrings {rendererClasses [], renderers []}
}
summary {
 systemName=NS_Bass_GeneAglet_Strings
 emitters [
  {emitterName=LightweightStrings, rendererClasses []}
  {emitterName=Light, rendererClasses [&NiagaraLightRendererProperties, &NiagaraSpriteRendererProperties]}
 ]
 userVariables []
}
